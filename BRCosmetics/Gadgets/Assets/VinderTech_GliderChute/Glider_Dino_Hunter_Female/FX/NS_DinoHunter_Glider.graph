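# Niagara effect for the female Dino Hunter glider, shedding mesh feathers and emitting wind streak sprites and two ribbon anim trails.
# Particles spawn from socket pairs on the user skeletal mesh, with SkeletalMeshIndex choosing the feather visibility tag.
# An FN_Integrator value scales wind rates, speeds and alpha, while bIsFullyDeployed gates streak spawning.
# Front-end modifiers adjust counts, offsets and velocities outside gameplay; the Rushing_Wind emitter is disabled.
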NiagaraSystem /BRCosmetics/Gadgets/Assets/VinderTech_GliderChute/Glider_Dino_Hunter_Female/FX/NS_DinoHunter_Glider
dependencies {
 usedDataInterfaces [&NiagaraDataInterfaceCurve, &NiagaraDataInterfaceColorCurve, &NiagaraDataInterfaceArrayFloat]
 usedDynamicInputs [&ClampFloat]
 usedModules [&SystemState, &FN_Integrator]
}
emitters {
 Feathers {
  emitterUpdateScript [
   &EmitterState(
    Life Cycle Mode=::NewEnumerator1
    Loop Behavior=::NewEnumerator1
    Loop Delay=&RandomRangeFloat(Minimum=0.125, Maximum=0.5)
    Loop Duration=1
    UseLoopDelay=false
   )
   &SpawnBurst_Instantaneous(
    Spawn Count=&MakeIntFromFloat(Float=&FN_FrontEndFloatModifier(Gameplay=18, Front End=9))
    Use Spawn Probability=false
   )
  ]
  particleSpawnScript [
   /SetVariables(Particles.VisibilityTag=&MakeIntFromFloat(Float=&SelectFloatFromArray(
      Direct Array Index=&MakeIntFromFloat(Float=$User.SkeletalMeshIndex)
      Float Selection Array=DataInterface() {propertyValues {FloatData [1,2]}}
     ))) {isSetParameters=true}
   &InitializeParticle(
    Color Mode=::NewEnumerator0
    Lifetime=&RandomRangeFloat(Minimum=1.25, Maximum=1.56)
    Mesh Scale Mode=::NewEnumerator3
    Mesh Uniform Scale=&FN_FrontEndFloatModifier(
     Gameplay=&RandomRangeFloat(Minimum=0.6, Maximum=1.4)
     Front End=&RandomRangeFloat(Minimum=0.5, Maximum=0.8)
    )
    Sprite Size=<1.25 120>
    Sprite Size Mode=::NewEnumerator4
    UsePositionOffset=false
    Write Material Random=true
   )
   &FN_SetSpawnGroup(Number of Spawn Groups=2)
   &SkeletalMeshLocation(
    Apply To Particle Position=false
    Bone / Socket Index=HlslExpression() {hlslExpression=Particles.SpawnGroup * 2}
    Inherit Velocity=false
    MaskSpawnGroup=false
    Mesh Sampling Type=::NewEnumerator5
    Sampled Position Offset Coordinate Space:ENiagaraMeshTransforms=::NewEnumerator3
    Skeletal Mesh=$User.SkeletalMesh
    Socket Sampling Mode=::NewEnumerator5
   )
   &SkeletalMeshLocation(
    Apply To Particle Position=false
    Bone / Socket Index=HlslExpression() {hlslExpression="(Particles.SpawnGroup * 2) + 1"}
    Inherit Velocity=false
    MaskSpawnGroup=false
    Mesh Sampling Type=::NewEnumerator5
    Sampled Position Offset=<0 63.0028 0>
    Sampled Position Offset Coordinate Space:ENiagaraMeshTransforms=::NewEnumerator3
    Skeletal Mesh=$User.SkeletalMesh
    Socket Sampling Mode=::NewEnumerator5
   ) {module=SkeletalMeshLocation001}
   /SetVariables(Particles.Position=&Lerp_Vector(
     A=&Add_Vector(
      A=$Output.SkeletalMeshLocation.SampledPosition
      B=&MakeVector(
       X=0
       Y=&MakeCustomFloatFromBool(Bool=HlslExpression() {hlslExpression="User.ForwardDot > 0.5"}, True Float=0, False Float=0)
       Z=0
      )
     )
     B=&Add_Vector(
      A=$Output.SkeletalMeshLocation001.SampledPosition
      B=&MakeVector(
       X=0
       Y=&MakeCustomFloatFromBool(Bool=HlslExpression() {hlslExpression="User.ForwardDot > 0.5"}, True Float=0, False Float=0)
       Z=0
      )
     )
     Alpha=&RandomRangeFloat(Minimum=0.25, Maximum=1)
    )) {isSetParameters=true, module=SetVariables_#1}
   &AddVelocity(Velocity=&FN_FrontEndVector3Modifier(
     Gameplay=&RandomRangeVector(Minimum=<-150 100 50>, Maximum=<150 950 200>)
     Front End=&RandomRangeVector(Minimum=<-50 0 -50>, Maximum=<50 -300 50>)
    ))
   &InitialMeshOrientation(
    Orientation Axis=<0 -1 0>
    Orientation Vector=&RandomRangeVector(Minimum=<-1 -1 -1>, Maximum=<1 1 1>)
    Use Orientation Vector=true
    Use Rotation Vector=false
   )
  ]
  particleUpdateScript [
   &ParticleState()
   &Drag(Use Linear Drag=true, Use Rotational Drag=false)
   &UpdateMeshOrientation(
    Rotation Rate=&Multiply_Float(
     A=1
     B=&FloatFromCurve(
      FloatCurve=DataInterface() {propertyValues {Curve {keys [<at 0 5>,<at 1 1>]}}}
      Scale Curve=&RandomRangeFloat(Minimum=0.2, Maximum=1)
     )
    )
    Rotation Vector=&RandomRangeVector(Minimum=<-1 -1 -1>, Maximum=<1 1 1>)
   )
   &ScaleMeshSize(Scale Factor=&VectorFromFloat(Value=&FloatFromCurve(FloatCurve=DataInterface() {propertyValues {Curve {keys [
          {interpMode=RCIM_Cubic, tangentMode=RCTM_Break, time=0, value=1}
          {arriveTangent=-3.4436, interpMode=RCIM_Cubic, tangentMode=RCTM_Break, time=1, value=0}
         ]}}})))
   &InheritSourceMovement(Applied Position Scale=<1 0.9 1>, ApplyForce=false, ApplyPosition=true, ApplyPositionFalloff=false)
   &SolveForcesAndVelocity(Clamp Velocity=false, Limit Acceleration=false)
  ]
  rendererClasses [&NiagaraMeshRendererProperties]
  renderers [
   {rendererClass=&NiagaraMeshRendererProperties, rendererIndex=0}
   {rendererClass=&NiagaraMeshRendererProperties, rendererIndex=1}
  ]
 }
 Rushing_Wind {
  bEnabled=false
  emitterUpdateScript [
   &EmitterState(Life Cycle Mode=::NewEnumerator1, Loop Delay=3, Loop Duration=1, UseLoopDelay=false)
   &SpawnRate(
    Spawn Probability=&MakeCustomFloatFromBool(Bool=$User.bIsFullyDeployed, True Float=1, False Float=0)
    SpawnRate=&Lerp_Float(A=50, B=65, Alpha=$System.FN_Integrator.Value)
    Use Spawn Probability=true
   )
  ]
  particleSpawnScript [
   &InitializeParticle(
    Color Mode=::NewEnumerator0
    Lifetime=1.25
    Sprite Rotation Angle=&RandomRangeFloat(Minimum=0, Maximum=360)
    Sprite Rotation Mode=::NewEnumerator0
    Sprite Size Mode=::NewEnumerator1
    Uniform Sprite Size Max=175
    Uniform Sprite Size Min=100
    UsePositionOffset=false
    Write Material Random=true
   )
   &FN_SetSpawnGroup(Number of Spawn Groups=2)
   &SkeletalMeshLocation(
    Apply To Particle Position=false
    Bone / Socket Index=HlslExpression() {hlslExpression=Particles.SpawnGroup * 2}
    Inherit Velocity=false
    MaskSpawnGroup=false
    Mesh Sampling Type=::NewEnumerator5
    Sampled Position Offset=<0 14199.8291 0>
    Sampled Position Offset Coordinate Space:ENiagaraMeshTransforms=::NewEnumerator3
    Skeletal Mesh=$User.SkeletalMesh
    Socket Sampling Mode=::NewEnumerator5
   )
   &SkeletalMeshLocation(
    Apply To Particle Position=false
    Bone / Socket Index=HlslExpression() {hlslExpression="(Particles.SpawnGroup * 2) + 1"}
    Inherit Velocity=false
    MaskSpawnGroup=false
    Mesh Sampling Type=::NewEnumerator5
    Sampled Position Offset=<0 63.0028 0>
    Sampled Position Offset Coordinate Space:ENiagaraMeshTransforms=::NewEnumerator3
    Skeletal Mesh=$User.SkeletalMesh
    Socket Sampling Mode=::NewEnumerator5
   ) {module=SkeletalMeshLocation001}
   /SetVariables(Particles.Position=&Lerp_Vector(
     A=&Add_Vector(
      A=$Output.SkeletalMeshLocation.SampledPosition
      B=&MakeVector(
       X=0
       Y=&MakeCustomFloatFromBool(Bool=HlslExpression() {hlslExpression="User.ForwardDot > 0.5"}, True Float=10, False Float=30)
       Z=0
      )
     )
     B=&Add_Vector(
      A=$Output.SkeletalMeshLocation001.SampledPosition
      B=&MakeVector(
       X=0
       Y=&MakeCustomFloatFromBool(Bool=HlslExpression() {hlslExpression="User.ForwardDot > 0.5"}, True Float=10, False Float=30)
       Z=0
      )
     )
     Alpha=&RandomRangeFloat(Minimum=0.25, Maximum=1)
    )) {isSetParameters=true, module=SetVariables_#1}
   &AddVelocity(Velocity=&MakeVector(
     X=&FN_FrontEndFloatModifier(Gameplay=0, Front End=-125)
     Y=&FN_FrontEndFloatModifier(
      Gameplay=&MakeCustomFloatFromBool(
       Bool=$User.bIsFullyDeployed
       True Float=&Lerp_Float(A=350, B=650, Alpha=$System.FN_Integrator.Value)
       False Float=0
      )
      Front End=115
     )
     Z=-50
    ))
  ]
  particleUpdateScript [
   &ParticleState()
   &SolveForcesAndVelocity(Clamp Velocity=false, Limit Acceleration=false)
   &SubUVAnimation(Start Frame:NiagaraInt32=&RandomRangeInt(Minimum=0, Maximum=45))
   &DynamicMaterialParameters(
    Index 0 Param 1=&FloatFromCurve(FloatCurve=DataInterface() {propertyValues {Curve {keys [<at 0 1>,<at 1 -1>]}}})
    Index 0 Param 4=2
    Module.Param0WriteEnabled=true
    Module.Param1WriteEnabled=true
    Module.Param2WriteEnabled=true
    Module.Param3WriteEnabled=true
   )
   &Color(
    Color=&ColorFromCurve(ColorCurve=DataInterface() {propertyValues {
       AlphaCurve {keys [<at 0 0 RCIM_Cubic>,<at 0.1165 1>,<at 0.4375 1>,<at 0.8888 0 RCIM_Cubic>]}
       BlueCurve {keys [<at 0 1 RCIM_Cubic>,<at 1 1 RCIM_Cubic>]}
       GreenCurve {keys [<at 0 1 RCIM_Cubic>,<at 1 1 RCIM_Cubic>]}
       RedCurve {keys [<at 0 1 RCIM_Cubic>,<at 1 1 RCIM_Cubic>]}
      }})
    Scale Alpha=&FloatFromCurve(
     FloatCurve=DataInterface() {propertyValues {Curve {keys [<at 0 1>,<at 1 0>]}}}
     Scale Curve=&Lerp_Float(A=0, B=0.02, Alpha=$System.FN_Integrator.Value)
    )
    Scale Color=<0.6 0.6 0.6>
   )
  ]
  rendererClasses [&NiagaraSpriteRendererProperties]
  renderers [{rendererClass=&NiagaraSpriteRendererProperties}]
 }
 Wind_Streaks {
  emitterUpdateScript [
   &EmitterState(
    Delay First Loop Only=true
    Life Cycle Mode=::NewEnumerator1
    Loop Delay=&RandomRangeFloat(Minimum=0.125, Maximum=0.5)
    Loop Duration=1
    UseLoopDelay=true
   )
   &SpawnRate(
    Spawn Probability=&MakeCustomFloatFromBool(Bool=$User.bIsFullyDeployed, True Float=1, False Float=0)
    SpawnRate=&Lerp_Float(A=10, B=20, Alpha=$System.FN_Integrator.Value)
    Use Spawn Probability=true
   )
  ]
  particleSpawnScript [
   &InitializeParticle(Color Mode=::NewEnumerator0, Sprite Size=<1.25 120>, Sprite Size Mode=::NewEnumerator4, UsePositionOffset=false, Write Material Random=true)
   &FN_SetSpawnGroup(Number of Spawn Groups=2)
   &SkeletalMeshLocation(
    Apply To Particle Position=false
    Bone / Socket Index=HlslExpression() {hlslExpression=Particles.SpawnGroup * 2}
    Inherit Velocity=false
    MaskSpawnGroup=false
    Mesh Sampling Type=::NewEnumerator5
    Sampled Position Offset Coordinate Space:ENiagaraMeshTransforms=::NewEnumerator3
    Skeletal Mesh=$User.SkeletalMesh
    Socket Sampling Mode=::NewEnumerator5
   )
   &SkeletalMeshLocation(
    Apply To Particle Position=false
    Bone / Socket Index=HlslExpression() {hlslExpression="(Particles.SpawnGroup * 2) + 1"}
    Inherit Velocity=false
    MaskSpawnGroup=false
    Mesh Sampling Type=::NewEnumerator5
    Sampled Position Offset=<0 63.0028 0>
    Sampled Position Offset Coordinate Space:ENiagaraMeshTransforms=::NewEnumerator3
    Skeletal Mesh=$User.SkeletalMesh
    Socket Sampling Mode=::NewEnumerator5
   ) {module=SkeletalMeshLocation001}
   /SetVariables(Particles.Position=&Lerp_Vector(
     A=&Add_Vector(
      A=$Output.SkeletalMeshLocation.SampledPosition
      B=&MakeVector(
       X=0
       Y=&MakeCustomFloatFromBool(Bool=HlslExpression() {hlslExpression="User.ForwardDot > 0.5"}, True Float=0, False Float=0)
       Z=0
      )
     )
     B=&Add_Vector(
      A=$Output.SkeletalMeshLocation001.SampledPosition
      B=&MakeVector(
       X=0
       Y=&MakeCustomFloatFromBool(Bool=HlslExpression() {hlslExpression="User.ForwardDot > 0.5"}, True Float=0, False Float=0)
       Z=0
      )
     )
     Alpha=&RandomRangeFloat(Minimum=0.25, Maximum=1)
    )) {isSetParameters=true, module=SetVariables_#1}
   &AddVelocity(Velocity=&MakeVector(
     X=&FN_FrontEndFloatModifier(Gameplay=0, Front End=-125)
     Y=&FN_FrontEndFloatModifier(
      Gameplay=&MakeCustomFloatFromBool(
       Bool=$User.bIsFullyDeployed
       True Float=&Lerp_Float(A=1150, B=650, Alpha=$System.FN_Integrator.Value)
       False Float=0
      )
      Front End=115
     )
     Z=-50
    ))
  ]
  particleUpdateScript [
   &ParticleState()
   &Drag(Drag=3, Use Linear Drag=true, Use Rotational Drag=false)
   &SolveForcesAndVelocity(Clamp Velocity=false, Limit Acceleration=false)
   &Color(
    Color=&ColorFromCurve(ColorCurve=DataInterface() {propertyValues {
       AlphaCurve {keys [<at 0 0 RCIM_Cubic>,<at 0.1165 1>,<at 0.4375 1>,<at 0.8888 0 RCIM_Cubic>]}
       BlueCurve {keys [<at 0 1 RCIM_Cubic>,<at 1 1 RCIM_Cubic>]}
       GreenCurve {keys [<at 0 1 RCIM_Cubic>,<at 1 1 RCIM_Cubic>]}
       RedCurve {keys [<at 0 1 RCIM_Cubic>,<at 1 1 RCIM_Cubic>]}
      }})
    Scale Alpha=&FloatFromCurve(
     FloatCurve=DataInterface() {propertyValues {Curve {keys [{interpMode=RCIM_Cubic, tangentMode=RCTM_Break, time=0, value=1}, {interpMode=RCIM_Cubic, tangentMode=RCTM_Break, time=1, value=0}]}}}
     Scale Curve=&Lerp_Float(A=0, B=0.1, Alpha=$System.FN_Integrator.Value)
    )
    Scale Color=<0.6 0.6 0.6>
   )
  ]
  rendererClasses [&NiagaraSpriteRendererProperties]
  renderers [{rendererClass=&NiagaraSpriteRendererProperties}]
 }
 Wind_Trail {
  emitterSpawnScript [&FN_RibbonIDOffsetUtility()]
  emitterUpdateScript [
   &EmitterState()
   &FN_SpawnAnimTrail(Number of Anim Trails=2)
  ]
  particleSpawnScript [
   &FN_InitializeAnimTrailBoneIndecies(First Bone or Socket Index=1, Second Bone or Socket Index=0)
   &FN_InitializeAnimTrailBoneIndecies(AnimTrail Selection=1, First Bone or Socket Index=3, Second Bone or Socket Index=2) {module=FN_InitializeAnimTrailBoneIndecies001}
   /SetVariables(Particles.AlphaFadeIn:NiagaraFloat=/Niagara/DynamicInputs/Math/Min_Float(A:NiagaraFloat=&Divide_Float(A=$Emitter.Age, B=1), B:NiagaraFloat=1)) {isSetParameters=true}
   &FN_InitializeAnimTrail(
    Color=&MakeLinearColor(
     R:NiagaraFloat=1
     G:NiagaraFloat=1
     B:NiagaraFloat=1
     A:NiagaraFloat=&Multiply_Float(A=$Particles.AlphaFadeIn, B=0.5)
    )
    Lifetime=&MakeCustomFloatFromBool(Bool=$Particles.Locked, True Float=100, False Float=1.5)
    Position Offset=&FN_FrontEndVector3Modifier(Gameplay=<0 -20 0>, Front End=<0 30 0>)
   )
   &AddVelocity(Velocity=&Multiply_VectorByFloat(
     Vector=&MakeVector(X=0, Y=&Lerp_Float(A=100, B=1000, Alpha=$System.FN_Integrator.Value), Z=0)
     Float=&MakeCustomFloatFromBool(Bool=$User.bIsFrontEnd, True Float=-1, False Float=1)
    ))
  ]
  particleUpdateScript [
   &SolveForcesAndVelocity(Clamp Velocity=false, Limit Acceleration=false)
   &FN_UpdateAnimTrail(
    Locked Particle Color:LinearColor=&MultiplyLinearColor(
     Linear Color=$Particles.Initial.Color
     Scale Factor=&MakeFloat4FromVectorAndFloat("Vector (RGB)"=<1 1 1>, "Float (Alpha)"=$System.FN_Integrator.Value)
    )
    Locked Particle RibbonWidth=1
    SkeletalMesh=$User.SkeletalMesh
   ) {module=FN_UpdateAnimTrail001}
   /SetVariables(Particles.RibbonU0Override=$Particles.NormalizedAge) {isSetParameters=true}
   &FN_RibbonMaterial(
    Opacity Pan U Speed=0.5
    Opacity Scale U=0.1
    Scale Alpha:NiagaraFloat=&FloatFromCurve(
     FloatCurve=DataInterface() {propertyValues {Curve {keys [<at 0 1>,<at 1 0>]}}}
     Scale Curve=&Lerp_Float(A=0, B=0.15, Alpha=$System.FN_Integrator.Value)
    )
   )
  ]
  rendererClasses [&NiagaraRibbonRendererProperties]
  renderers [{rendererClass=&NiagaraRibbonRendererProperties}]
 }
}
summary {
 systemName=NS_DinoHunter_Glider
 emitters [
  {bEnabled=false, emitterName=Rushing_Wind, rendererClasses [&NiagaraSpriteRendererProperties]}
  {emitterName=Wind_Streaks, rendererClasses [&NiagaraSpriteRendererProperties]}
  {emitterName=Feathers, rendererClasses [&NiagaraMeshRendererProperties]}
  {emitterName=Wind_Trail, rendererClasses [&NiagaraRibbonRendererProperties]}
 ]
 userVariables [
  User.bIsFrontEnd {type=NiagaraBool}
  User.bIsFrontEndPreview {type=NiagaraBool}
  User.bIsFullyDeployed {type=NiagaraBool}
  User.ForwardDot {type=NiagaraFloat, defaultValue=0, description=""}
  User.RibbonIDOffset {type=NiagaraInt32}
  User.SkeletalMesh {
   type=NiagaraDataInterfaceSkeletalMesh
   defaultValue {
    struct=NiagaraExt_VariableValue_DataInterface
    value {dataInterface=&self:NiagaraDataInterfaceSkeletalMesh_0}
   }
  }
  User.SkeletalMeshIndex {type=NiagaraFloat, defaultValue=0, description=""}
 ]
}
